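# Physics asset for the EthicalNoggin pickaxe, a two-handed melee weapon cosmetic.
# It holds a single capsule body on the Melee_TwoHanded_Handle bone, offset mainly along Z and rotated 90 degrees in yaw.
# It has no constraints, so this one broad capsule is the weapon's entire physics shape.
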PhysicsAsset /BRCosmetics/Weapons/FORT_Melee/Pickaxe_EthicalNoggin/EthicalNoggin_Axe_PhysicsAsset
bodies {Melee_TwoHanded_Handle [{shapeType=Capsule, length=66.1193, radius=33.6326, center=<0.9926 2.5608 28.2835>, rotation=<rot 0.3154 -90 0.0037>}]}
constraints []
